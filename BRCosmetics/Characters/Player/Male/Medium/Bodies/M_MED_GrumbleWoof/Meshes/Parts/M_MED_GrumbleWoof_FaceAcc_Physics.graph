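# Physics asset for the GrumbleWoof male character's face accessory mesh, defining capsule bodies and joints for a left and right scarf.
# Each side chains a small base capsule and three dynamic segments, with the base jointed to spine_05 with free rotation.
# Segment joints keep swing limits tight and twist locked or nearly so.
# Larger capsules on spine_05 and the unconstrained right upper-arm twist bone appear to act as body colliders.
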PhysicsAsset /BRCosmetics/Characters/Player/Male/Medium/Bodies/M_MED_GrumbleWoof/Meshes/Parts/M_MED_GrumbleWoof_FaceAcc_Physics
bodies {
 base_scarf_lf [{shapeType=Capsule, length=0.505, radius=0.505, shapeName=base_scarf_lf_capsule, rotation=<rot 0 -90 0>}]
 base_scarf_rt [{shapeType=Capsule, length=0.505, radius=0.505, shapeName=base_scarf_rt_capsule, rotation=<rot 0 -90 0>}]
 dyn_scarf_01_lf [{shapeType=Capsule, length=3, radius=1.5, shapeName=dyn_scarf_01_lf_capsule, center=<5 0 0>, rotation=<rot 90 0 0>}]
 dyn_scarf_01_rt [{shapeType=Capsule, length=2.6853, radius=1.8147, shapeName=dyn_scarf_01_rt_capsule, center=<2 0 0>, rotation=<rot 90 0 0>}]
 dyn_scarf_02_lf [{shapeType=Capsule, length=3, radius=1.5, shapeName=dyn_scarf_02_lf_capsule, center=<5 0 0>, rotation=<rot 90 0 0>}]
 dyn_scarf_02_rt [{shapeType=Capsule, length=5.7404, radius=1.7596, shapeName=dyn_scarf_02_rt_capsule, center=<2 -0.1509 -0.6537>, rotation=<rot 0 0 103>}]
 dyn_scarf_03_lf [{shapeType=Capsule, length=3, radius=1.5, shapeName=dyn_scarf_03_lf_capsule, center=<5 0 0>, rotation=<rot 90 0 0>}]
 dyn_scarf_03_rt [{
   shapeType=Capsule
   length=6.4467
   radius=1.7912
   shapeName=dyn_scarf_03_rt_capsule
   center=<1.989 -0.8917 0.6812>
   rotation=<rot 0.2889 -6.3351 77.7492>
  }]
 spine_05 [{shapeType=Capsule, length=14.0321, radius=13.4079, center=<0 -2.1948 0>}]
 upperarm_twist_01_r [{
   shapeType=Capsule
   length=24.1609
   radius=6.1248
   shapeName=upperarm_twist_01_r_capsule
   center=<-8.2092 0.8103 0.2383>
   rotation=<rot 90 0 0>
  }]
}
constraints [
 {
  bone1Name=base_scarf_lf
  bone2Name=spine_05
  swing1LimitDegrees=45
  swing1Motion=Free
  swing2LimitDegrees=45
  swing2Motion=Free
  twistLimitDegrees=45
  twistMotion=Free
 }
 {bone1Name=dyn_scarf_01_lf, bone2Name=base_scarf_lf, swing1LimitDegrees=20, swing2LimitDegrees=15, twistLimitDegrees=1}
 {bone1Name=dyn_scarf_02_lf, bone2Name=dyn_scarf_01_lf, swing1LimitDegrees=15, swing2LimitDegrees=20, twistLimitDegrees=1}
 {bone1Name=dyn_scarf_03_lf, bone2Name=dyn_scarf_02_lf, swing1LimitDegrees=10, swing2LimitDegrees=15, twistLimitDegrees=1}
 {
  bone1Name=base_scarf_rt
  bone2Name=spine_05
  swing1LimitDegrees=45
  swing1Motion=Free
  swing2LimitDegrees=45
  swing2Motion=Free
  twistLimitDegrees=45
  twistMotion=Free
 }
 {bone1Name=dyn_scarf_01_rt, bone2Name=base_scarf_rt, swing1LimitDegrees=10, swing2LimitDegrees=10, twistLimitDegrees=45, twistMotion=Locked}
 {bone1Name=dyn_scarf_02_rt, bone2Name=dyn_scarf_01_rt, swing1LimitDegrees=20, swing2LimitDegrees=20, twistLimitDegrees=45, twistMotion=Locked}
 {bone1Name=dyn_scarf_03_rt, bone2Name=dyn_scarf_02_rt, swing1LimitDegrees=10, swing2LimitDegrees=15, twistLimitDegrees=45, twistMotion=Locked}
]
