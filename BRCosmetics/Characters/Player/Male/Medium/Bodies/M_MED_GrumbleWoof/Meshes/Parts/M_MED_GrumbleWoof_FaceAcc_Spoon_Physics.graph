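# Physics asset for the spoon accessory part (FaceAcc) of the GrumbleWoof male character body, defining its simulated dangle.
# Capsule bodies cover a small base holder, a dynamic holder, three spoons and a left-thigh helper, likely a collider for the spoons.
# The holder is constrained to the base, and each spoon attaches to the holder in parallel rather than in a chain.
# All joints lock twist and have narrow swing limits, varied per spoon.
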PhysicsAsset /BRCosmetics/Characters/Player/Male/Medium/Bodies/M_MED_GrumbleWoof/Meshes/Parts/M_MED_GrumbleWoof_FaceAcc_Spoon_Physics
bodies {
 base_spoon_holder [{shapeType=Capsule, length=0.505, radius=0.505, shapeName=base_spoon_holder_capsule, rotation=<rot 0 -90 0>}]
 dyn_spoon_01 [{
   shapeType=Capsule
   length=18.4128
   radius=1.9756
   shapeName=dyn_spoon_01_capsule
   center=<9.0428 -1.0755 0.4973>
   rotation=<rot 0.1265 84.6487 -86.8685>
  }]
 dyn_spoon_02 [{
   shapeType=Capsule
   length=14.3456
   radius=1.5938
   shapeName=dyn_spoon_02_capsule
   center=<6.7954 -0.2564 -0.012>
   rotation=<rot 0 88.1494 -87.7338>
  }]
 dyn_spoon_03 [{
   shapeType=Capsule
   length=12.4917
   radius=1.277
   shapeName=dyn_spoon_03_capsule
   center=<5.9095 0.0795 -0.1221>
   rotation=<rot 0 90.8215 -89.1248>
  }]
 dyn_spoon_holder [{
   shapeType=Capsule
   length=7.7459
   radius=1.9324
   shapeName=dyn_spoon_holder_capsule
   center=<3.7059 -1.2517 -0.267>
   rotation=<rot 5.99 73.2706 79.0556>
  }]
 helper_thigh_l [{
   shapeType=Capsule
   length=22.9586
   radius=8.5579
   shapeName=thigh_l_capsule
   center=<-7.0817 -0.4207 1.0092>
   rotation=<rot 82.3061 111.6542 111.6308>
  }]
}
constraints [
 {
  bone1Name=dyn_spoon_holder
  bone2Name=base_spoon_holder
  swing1LimitDegrees=10
  swing2LimitDegrees=10
  twistLimitDegrees=45
  twistMotion=Locked
 }
 {bone1Name=dyn_spoon_01, bone2Name=dyn_spoon_holder, swing1LimitDegrees=10, swing2LimitDegrees=10, twistLimitDegrees=45, twistMotion=Locked}
 {bone1Name=dyn_spoon_02, bone2Name=dyn_spoon_holder, swing1LimitDegrees=20, swing2LimitDegrees=5, twistLimitDegrees=45, twistMotion=Locked}
 {bone1Name=dyn_spoon_03, bone2Name=dyn_spoon_holder, swing1LimitDegrees=25, swing2LimitDegrees=10, twistLimitDegrees=45, twistMotion=Locked}
]
