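# Post-copy animation Blueprint for the GrumbleWoof male character's face accessory part, including belt, spoon and chain secondary motion.
# It blends parent and body mesh poses per bone, constrains belt bones to pelvis, spine and thighs, then runs a Control Rig, additive bone offsets and three RigidBody simulations.
# The event graph caches the body mesh and updates gravity override and rig flags; IDLE is true during movement, airborne, crouch, emote or DBNO states.
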AnimBlueprint /BRCosmetics/Characters/Player/Male/Medium/Bodies/M_MED_GrumbleWoof/Meshes/Parts/M_MED_GrumbleWoof_FaceAcc_AnimBP
target_skeleton=&/M_MED_GrumbleWoof_FaceAcc_Skeleton
graphs [
 AnimGraph {nodes [
   n1 = "Constraint - base_belt_fr_00 (PPPPPP)"!(
    ComponentPose=Local To Component(LocalPose=Use cached pose 'Start_Cache'())
    Alpha=1
    ConstraintWeights_0=0.8144
    ConstraintWeights_1:float=0.0705
    ConstraintWeights_2:float=0.0605
    ConstraintWeights_3:float=0.0413
    ConstraintWeights_4:float=0.0074
    ConstraintWeights_5:float=0.0059
   ) -> (Pose) {
    alphaBoolBlend {blendOption=Linear}
    boneToModify {boneName=base_belt_fr_00}
    constraintSetup [
     {targetBone {boneName=pelvis}}
     {targetBone {boneName=spine_02}}
     {targetBone {boneName=spine_01}}
     {targetBone {boneName=spine_03}}
     {targetBone {boneName=thigh_twist_01_l}}
     {targetBone {boneName=thigh_twist_01_r}}
    ]
    constraintWeights [1,1,1,1,1,1]
   }
   n2 = "Constraint - base_belt_side_00 (PPPPPPP)"!(
    ComponentPose=@n1
    Alpha=1
    ConstraintWeights_0=0.7485
    ConstraintWeights_1:float=0.1171
    ConstraintWeights_2:float=0.066
    ConstraintWeights_3:float=0.0394
    ConstraintWeights_4:float=0.0223
    ConstraintWeights_5:float=0.0065
    ConstraintWeights_6:float=0.0001
   ) -> (Pose) {
    alphaBoolBlend {blendOption=Linear}
    boneToModify {boneName=base_belt_side_00}
    constraintSetup [
     {targetBone {boneName=pelvis}}
     {targetBone {boneName=spine_01}}
     {targetBone {boneName=spine_02}}
     {targetBone {boneName=spine_03}}
     {targetBone {boneName=thigh_twist_01_l}}
     {targetBone {boneName=thigh_twist_01_r}}
     {targetBone {boneName=base_belt_ref}}
    ]
    constraintWeights [1,1,1,1,1,1,1]
   }
   n3 = Control Rig!(
    Source=Component To Local(ComponentPose=@n2)
    Alpha
    IsCrouching:bool=Get bIsCrouching()
    IDLE:bool=Get IDLE() -> (IDLE:bool)
    RBAN OFF:bool=Get RBAN OFF() -> (RBAN OFF:bool)
   ) -> (Pose) {
    alphaBoolBlend {blendOption=Linear}
    controlRigAssetReference {blueprintRigClass=&/M_MED_GrumbleWoof_FaceAcc_CtrlRig.M_MED_GrumbleWoof_FaceAcc_CtrlRig_C}
   }
   n4 = "Transform (Modify) Bone - Bone: helper_thigh_l"!(ComponentPose=Local To Component(LocalPose=@n3), bAlphaBoolEnabled:bool=True, Translation="0,-1,0", Rotation="0,0,0", Scale="1,1,1") -> (Pose) {
    alphaInputType=Bool
    translationMode=BMM_Additive
    alphaBoolBlend {blendOption=Linear}
    boneToModify {boneName=helper_thigh_l}
    translation {y=-1}
   }
   n5 = "Transform (Modify) Bone - Bone: base_spoon_holder"!(ComponentPose=@n4, bAlphaBoolEnabled:bool=True, Translation="0,0,0", Rotation="41.896,-26.942,-21.495", Scale="1,1,1") -> (Pose) {
    alphaInputType=Bool
    rotationMode=BMM_Additive
    rotationSpace=BCS_BoneSpace
    translationMode=BMM_Additive
    alphaBoolBlend {blendInTime=0.1, blendOption=Linear, blendOutTime=0.1}
    boneToModify {boneName=base_spoon_holder}
    rotation=<rot 41.896 -26.942 -21.495>
   }
   n6 = "Transform (Modify) Bone - Bone: base_spoon_holder"(ComponentPose=@n5, bAlphaBoolEnabled=True, Translation="0,0,0", Rotation="-15,-15,0", Scale="1,1,1") {
    alphaInputType=Bool
    rotationMode=BMM_Additive
    rotationSpace=BCS_BoneSpace
    translationMode=BMM_Additive
    alphaBoolBlend {blendOption=Linear}
    boneToModify {boneName=base_spoon_holder}
    rotation {pitch=-15, yaw=-15}
   }
   n7 = "Transform (Modify) Bone - Bone: base_belt_fr_00"!(ComponentPose=@n6, bAlphaBoolEnabled:bool=True, Translation="0,0,0", Rotation="28.514,-0.001,0.009", Scale="1,1,1") -> (Pose) {
    alphaInputType=Bool
    rotationMode=BMM_Additive
    rotationSpace=BCS_BoneSpace
    translationMode=BMM_Additive
    translationSpace=BCS_BoneSpace
    alphaBoolBlend {blendOption=Linear}
    boneToModify {boneName=base_belt_fr_00}
    rotation=<rot 28.514 -0.001 0.009>
   }
   n8 = RigidBody!(ComponentPose=@n7, Alpha, OverrideWorldGravity:Vector="0,0,0", ExternalForce) -> (Pose) {
    bClampLinearTranslationLimitToRefPose=true
    bOverrideWorldGravity=true
    overridePhysicsAsset=&/M_MED_GrumbleWoof_FaceAcc_Physics
    simulationSpace=BaseBoneSpace
    alphaBoolBlend {blendOption=Linear}
    baseBoneRef {boneName=spine_05}
    componentLinearAccScale=<0.23 0.3 0.4>
    componentLinearVelScale=<0.23 0.3 0.4>
    simSpaceSettings {worldAlpha=0.25}
   }
   n9 = RigidBody!(ComponentPose=@n8, bAlphaBoolEnabled:bool, OverrideWorldGravity="0,0,0", ExternalForce) -> (Pose) {
    alphaInputType=Bool
    bClampLinearTranslationLimitToRefPose=true
    bOverrideWorldGravity=true
    overridePhysicsAsset=&/M_MED_GrumbleWoof_FaceAcc_Spoon_Physics
    simulationSpace=BaseBoneSpace
    alphaBoolBlend {blendInTime=0.1, blendOption=Linear, blendOutTime=0.1}
    baseBoneRef {boneName=pelvis}
    componentAppliedLinearAccClamp=<1000 1000 1000>
    componentLinearAccScale=<0.5 0.1 0.1>
    componentLinearVelScale=<0.5 0.1 0.1>
    simSpaceSettings {worldAlpha=0.1}
   }
   n10 = RigidBody(ComponentPose=@n9, OverrideWorldGravity="0,0,0") {
    alphaInputType=Bool
    bClampLinearTranslationLimitToRefPose=true
    bOverrideWorldGravity=true
    overridePhysicsAsset=&/M_MED_GrumbleWoof_FaceAcc_chain_Physics
    simulationSpace=BaseBoneSpace
    alphaBoolBlend {blendInTime=0.1, blendOption=Linear, blendOutTime=0.1}
    baseBoneRef {boneName=pelvis}
    componentAppliedLinearAccClamp=<1000 1000 1000>
    componentLinearAccScale=<0.5 0.1 0.1>
    componentLinearVelScale=<0.5 0.1 0.1>
    simSpaceSettings {worldAlpha=0.1}
   }
   Output Pose(Result=Component To Local(ComponentPose=@n10)) {layerGroup=None}
   n11 = Layered blend per bone(
    BasePose=Copy Pose From Mesh() {bUseAttachedParent=true}
    BlendPoses_0=Copy Pose From Mesh(SourceMeshComponent="Get Body Mesh "() -> ("Body Mesh ":Skeletal Mesh Component Object Reference)) {bUseAttachedParent=true}
   ) {layerSetup [{branchFilters [{boneName=thigh_l}, {boneName=upperarm_l}, {boneName=upperarm_r}]}]}
   Save cached pose 'Start_Cache'(Pose=@n11) {cacheName=Start_Cache}
  ]}
 EventGraph {nodes [
   Event BlueprintUpdateAnimation()
   > n1 = Sequence!(execute) -> (then_0, then_1, then_2:Exec, then_3:Exec)
   > n2 = Gravity Override State Function DEPRECATED(CharacterPartAnimInstance=Self-Reference(), GameTimeInSeconds=GetGameTimeInSeconds(), GravityOverrideParams=Get Gravity Override Params())
   > Set Gravity Override(Gravity Override=@n2)
   Set isNotSprinting(execute=@n1.then_1, isNotSprinting:bool=NOT Boolean(A=Get bIsSprinting())) -> (Output_Get:bool)
   Set IDLE(
    execute=@n1.then_2
    IDLE:bool=OR Boolean!(
     A=Get bIsJumping()
     B=Get bIsMoving2D()
     C:bool=Get bIsSkyDiving()
     D:bool=Get bIsCrouching()
     E:bool=Get bIsPlayingEmote()
     F:bool=Get bIsFalling()
     G:bool=Get bIsDBNO()
    ) -> (ReturnValue)
   ) -> (Output_Get:bool)
   Set RBAN OFF(execute=@n1.then_3, RBAN OFF:bool=NOT Boolean(A=Get bRigidBodyEnabled())) -> (Output_Get:bool)
   Event OnCharacterCustomizationLoaded()
   > n3 = Cast To PlayerPawn_Athena(Object=TryGetPawnOwner())
   > "Set Body Mesh "("Body Mesh ":Skeletal Mesh Component Object Reference=GetSkeletalMeshForPartType(self=@n3)) -> (Output_Get:Skeletal Mesh Component Object Reference)
  ]}
]
variables [
 "Body Mesh " {type=SkeletalMeshComponent}
 Gravity Override {type=Vector}
 Gravity Override Params {type=GravityOverrideParamsStruct, default {onGround_Moving {gravityOverride=<0 0 -500>}}}
 IDLE {type=boolean}
 isNotSprinting {type=boolean}
 RBAN OFF {type=boolean}
 Skel Mesh {type=SkeletalMeshComponent}
 test {type=number}
]
